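# Sound cue for small-clawed biped companion jump foley, routed to the creature pets sound class with companion foley attenuation.
# A quality-level node picks a random pool: the first two slots share ten jump waves, the last two share a reduced set of the first four.
# A top-level modulator adds slight pitch and volume variation.
# It shares the biped footstep concurrency set to limit overlapping playback.
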SoundCue /CosmeticCompanions/Assets/Audio_Shared/Foley/Biped/Claws/Claws_Small/SC_MImosa_Biped_Claws_Small_Jump
root=Modulator_0
nodes [
 {id=Modulator_0, children [QualityLevel_0], properties {pitchMax=1.1, pitchMin=0.9, volumeMax=1, volumeMin=0.9}}
 {id=QualityLevel_0, children [Random_2,Random_2,Random_3,Random_3]}
 {
  id=Random_2
  children [
   WavePlayer_11
   WavePlayer_12
   WavePlayer_13
   WavePlayer_14
   WavePlayer_15
   WavePlayer_16
   WavePlayer_17
   WavePlayer_18
   WavePlayer_19
   WavePlayer_20
  ]
  properties {weights [1,1,1,1,1,1,1,1,1,1]}
 }
 {id=Random_3, children [WavePlayer_21, WavePlayer_25, WavePlayer_26, WavePlayer_27], properties {weights [1,1,1,1]}}
 {
  id=WavePlayer_11
  properties {soundWaveAssetPtr=/CosmeticCompanions/Assets/Audio_Shared/Foley/Biped/Claws/Claws_Small/Wavs/Mimosa_Biped_Claws_Small_Jump_01}
 }
 {
  id=WavePlayer_12
  properties {soundWaveAssetPtr=/CosmeticCompanions/Assets/Audio_Shared/Foley/Biped/Claws/Claws_Small/Wavs/Mimosa_Biped_Claws_Small_Jump_02}
 }
 {
  id=WavePlayer_13
  properties {soundWaveAssetPtr=/CosmeticCompanions/Assets/Audio_Shared/Foley/Biped/Claws/Claws_Small/Wavs/Mimosa_Biped_Claws_Small_Jump_03}
 }
 {
  id=WavePlayer_14
  properties {soundWaveAssetPtr=/CosmeticCompanions/Assets/Audio_Shared/Foley/Biped/Claws/Claws_Small/Wavs/Mimosa_Biped_Claws_Small_Jump_04}
 }
 {
  id=WavePlayer_15
  properties {soundWaveAssetPtr=/CosmeticCompanions/Assets/Audio_Shared/Foley/Biped/Claws/Claws_Small/Wavs/Mimosa_Biped_Claws_Small_Jump_05}
 }
 {
  id=WavePlayer_16
  properties {soundWaveAssetPtr=/CosmeticCompanions/Assets/Audio_Shared/Foley/Biped/Claws/Claws_Small/Wavs/Mimosa_Biped_Claws_Small_Jump_06}
 }
 {
  id=WavePlayer_17
  properties {soundWaveAssetPtr=/CosmeticCompanions/Assets/Audio_Shared/Foley/Biped/Claws/Claws_Small/Wavs/Mimosa_Biped_Claws_Small_Jump_07}
 }
 {
  id=WavePlayer_18
  properties {soundWaveAssetPtr=/CosmeticCompanions/Assets/Audio_Shared/Foley/Biped/Claws/Claws_Small/Wavs/Mimosa_Biped_Claws_Small_Jump_08}
 }
 {
  id=WavePlayer_19
  properties {soundWaveAssetPtr=/CosmeticCompanions/Assets/Audio_Shared/Foley/Biped/Claws/Claws_Small/Wavs/Mimosa_Biped_Claws_Small_Jump_09}
 }
 {
  id=WavePlayer_20
  properties {soundWaveAssetPtr=/CosmeticCompanions/Assets/Audio_Shared/Foley/Biped/Claws/Claws_Small/Wavs/Mimosa_Biped_Claws_Small_Jump_10}
 }
 {
  id=WavePlayer_21
  properties {soundWaveAssetPtr=&Mimosa_Biped_Claws_Small_Jump_01}
 }
 {
  id=WavePlayer_25
  properties {soundWaveAssetPtr=&Mimosa_Biped_Claws_Small_Jump_02}
 }
 {
  id=WavePlayer_26
  properties {soundWaveAssetPtr=&Mimosa_Biped_Claws_Small_Jump_03}
 }
 {
  id=WavePlayer_27
  properties {soundWaveAssetPtr=&Mimosa_Biped_Claws_Small_Jump_04}
 }
]
properties {
 attenuationSettings=/Game/Sounds/Attenuation/AS_Companion_Foley
 maxDistance=1200
 priority=0
 soundClassObject=&Fort_Creatures_Pets
 volumeMultiplier=0.4
 concurrencySet [/CosmeticCompanions/Assets/Audio_Shared/Foley/Biped/CONC_Mimosa_Biped_Footstep]
}
