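# Sound cue for the short slide foley of small biped claws on cosmetic companions, classed as a creature pet sound.
# A quality-level node sends its first two levels to a random pick of eight slide waves and its last two to the first four only.
# The cue plays quiet and pitched up, with companion foley attenuation and the shared biped footstep concurrency set.
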SoundCue /CosmeticCompanions/Assets/Audio_Shared/Foley/Biped/Claws/Claws_Small/SC_MImosa_Biped_Claws_Small_SlideShort
root=QualityLevel_0
nodes [
 {id=QualityLevel_0, children [Random_2,Random_2,Random_3,Random_3]}
 {
  id=Random_2
  children [WavePlayer_0,WavePlayer_1,WavePlayer_2,WavePlayer_3,WavePlayer_4,WavePlayer_5,WavePlayer_6,WavePlayer_7]
  properties {weights [1,1,1,1,1,1,1,1]}
 }
 {id=Random_3, children [WavePlayer_8, WavePlayer_9, WavePlayer_10, WavePlayer_17], properties {weights [1,1,1,1]}}
 {
  id=WavePlayer_0
  properties {soundWaveAssetPtr=/CosmeticCompanions/Assets/Audio_Shared/Foley/Biped/Claws/Claws_Small/Wavs/Mimosa_Biped_Claws_Small_Slide_Short_01}
 }
 {
  id=WavePlayer_1
  properties {soundWaveAssetPtr=/CosmeticCompanions/Assets/Audio_Shared/Foley/Biped/Claws/Claws_Small/Wavs/Mimosa_Biped_Claws_Small_Slide_Short_02}
 }
 {
  id=WavePlayer_10
  properties {soundWaveAssetPtr=/CosmeticCompanions/Assets/Audio_Shared/Foley/Biped/Claws/Claws_Small/Wavs/Mimosa_Biped_Claws_Small_Slide_Short_03}
 }
 {
  id=WavePlayer_17
  properties {soundWaveAssetPtr=/CosmeticCompanions/Assets/Audio_Shared/Foley/Biped/Claws/Claws_Small/Wavs/Mimosa_Biped_Claws_Small_Slide_Short_04}
 }
 {
  id=WavePlayer_2
  properties {soundWaveAssetPtr=&Mimosa_Biped_Claws_Small_Slide_Short_03}
 }
 {
  id=WavePlayer_3
  properties {soundWaveAssetPtr=&Mimosa_Biped_Claws_Small_Slide_Short_04}
 }
 {
  id=WavePlayer_4
  properties {soundWaveAssetPtr=/CosmeticCompanions/Assets/Audio_Shared/Foley/Biped/Claws/Claws_Small/Wavs/Mimosa_Biped_Claws_Small_Slide_Short_05}
 }
 {
  id=WavePlayer_5
  properties {soundWaveAssetPtr=/CosmeticCompanions/Assets/Audio_Shared/Foley/Biped/Claws/Claws_Small/Wavs/Mimosa_Biped_Claws_Small_Slide_Short_06}
 }
 {
  id=WavePlayer_6
  properties {soundWaveAssetPtr=/CosmeticCompanions/Assets/Audio_Shared/Foley/Biped/Claws/Claws_Small/Wavs/Mimosa_Biped_Claws_Small_Slide_Short_07}
 }
 {
  id=WavePlayer_7
  properties {soundWaveAssetPtr=/CosmeticCompanions/Assets/Audio_Shared/Foley/Biped/Claws/Claws_Small/Wavs/Mimosa_Biped_Claws_Small_Slide_Short_08}
 }
 {
  id=WavePlayer_8
  properties {soundWaveAssetPtr=&Mimosa_Biped_Claws_Small_Slide_Short_01}
 }
 {
  id=WavePlayer_9
  properties {soundWaveAssetPtr=&Mimosa_Biped_Claws_Small_Slide_Short_02}
 }
]
properties {
 attenuationSettings=/Game/Sounds/Attenuation/AS_Companion_Foley
 maxDistance=1200
 pitchMultiplier=1.3
 priority=0
 soundClassObject=&Fort_Creatures_Pets
 volumeMultiplier=0.18
 concurrencySet [/CosmeticCompanions/Assets/Audio_Shared/Foley/Biped/CONC_Mimosa_Biped_Footstep]
}
